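# One-shot Niagara burst effect for the AstroAnode backpack cosmetic, built from two cyan-blue emitters.
# Burst throws a batch of mesh and sprite particles from a small volume, with separate front-end and gameplay velocities, drag, upward gravity and a fading GlowColor curve.
# Minimal adds a short sprite flash that grows, fades and follows a sampled location on the system's skeletal mesh.
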NiagaraSystem /BRCosmetics/Accessories/FORT_Backpacks/Backpack_AstroAnode/FX/NS_Backpack_AstroAnode_Burst
dependencies {
 usedDataInterfaces [&NiagaraDataInterfaceCurve, &NiagaraDataInterfaceColorCurve, &NiagaraDataInterfaceSkeletalMesh]
 usedDynamicInputs [&FN_FrontEndFloatModifier, &RandomRangeFloat]
 usedModules [
  &self:SystemSpawnScript.SystemScriptSource.SystemScriptGraph.NiagaraNodeAssignment_0.SetVariables_#1
  &SystemState
 ]
}
emitters {
 Burst {
  emitterUpdateScript [
   &EmitterState()
   &SpawnBurst_Instantaneous(
    Spawn Count=&RandomRangeInt(Minimum=14, Maximum=18)
    Use Spawn Probability=false
    UseLoopCountLimit=false
   )
  ]
  particleSpawnScript [
   &InitializeParticle(
    Color=<rgba 0.046 0.8162 1.5 2>
    Lifetime Max=0.675
    Lifetime Min=0.4725
    Lifetime Mode=::NewEnumerator1
    Mesh Scale Mode=::NewEnumerator1
    Mesh Uniform Scale Max=1.1
    Mesh Uniform Scale Min=0.8
    Sprite Size Mode=::NewEnumerator3
    UsePositionOffset=false
    Write Material Random=true
   )
   &ShapeLocation(
    Box Midpoint=<0.5 0.6 0.5>
    Box Size=<20 20 30>
    MaskSpawnGroup=false
    Override Seed=false
    Rotation Coordinate Space=::World
    Shape Origin=$Engine.Emitter.SimulationPosition
    Shape Primitive=::NewEnumerator2
    Surface Only=false
    UseOwnerScale=true
   )
   &AddVelocity(
    Velocity=&FN_FrontEndVector3Modifier(
     Gameplay=&RandomRangeVector(Minimum=<-100 -50 -100>, Maximum=<100 -100 100>)
     Front End=&RandomRangeVector(Minimum=<-30 -15 -30>, Maximum=<30 -30 30>)
    )
    Velocity Speed Scale=1.25
   )
  ]
  particleUpdateScript [
   &Drag(Drag=6, Use Linear Drag=true, Use Rotational Drag=true)
   &GravityForce(Gravity=<0 0 150>)
   &SolveForcesAndVelocity(Clamp Velocity=false, Limit Acceleration=false)
   &ParticleState()
   &UpdateMeshOrientation(Rotation Rate=0.6, Rotation Vector=&RandomRangeVector(Minimum=<0 0 0>, Maximum=<1 1 1>))
   &ScaleMeshSize(Scale Factor=&VectorFromFloat(Value=&FloatFromCurve(FloatCurve=DataInterface() {propertyValues {Curve {keys [
          {interpMode=RCIM_Cubic, leaveTangent=0.5222, tangentMode=RCTM_Break, time=0, value=0.4}
          {arriveTangent=-0.245, interpMode=RCIM_Cubic, tangentMode=RCTM_Break, time=0.3, value=1}
          {arriveTangent=-0.111, interpMode=RCIM_Cubic, leaveTangent=-3.88, tangentMode=RCTM_Break, time=0.8, value=0.9}
          <at 1 0.4 RCIM_Cubic>
         ]}}})))
   &ScaleColor(
    Scale Alpha=&FloatFromCurve(FloatCurve=DataInterface() {propertyValues {Curve {keys [
         {interpMode=RCIM_Cubic, leaveTangent=-0.5242, tangentMode=RCTM_Break, time=0.3, value=1}
         {arriveTangent=0.6674, tangentMode=RCTM_User, time=1, value=0}
        ]}}})
    Scale RGB=<0 0 0>
    ScaleA=true
    ScaleRGB=true
   )
   /SetVariables(Particles.GlowColor=&ColorFromCurve(ColorCurve=DataInterface() {propertyValues {
       AlphaCurve {keys [
         <at 0.0011 0>
         <at 0.0955 0.4>
         {arriveTangent=-0.4422, interpMode=RCIM_Cubic, leaveTangent=-0.4422, time=0.4528, value=0.5}
         <at 1 0 RCIM_Cubic>
        ]}
       BlueCurve {keys [<at 0 2 RCIM_Cubic>,<at 1 0.4167 RCIM_Cubic>]}
       GreenCurve {keys [<at 0 0.8564 RCIM_Cubic>,<at 1 0.0401 RCIM_Cubic>]}
       RedCurve {keys [<at 0 0 RCIM_Cubic>,<at 1 0.0561 RCIM_Cubic>]}
      }})) {isSetParameters=true}
  ]
  rendererClasses [&NiagaraMeshRendererProperties, &NiagaraSpriteRendererProperties]
  renderers [
   {rendererClass=&NiagaraMeshRendererProperties, rendererIndex=0}
   {rendererClass=&NiagaraSpriteRendererProperties, rendererIndex=1}
  ]
 }
 Minimal {
  emitterUpdateScript [
   &EmitterState()
   &SpawnBurst_Instantaneous(Use Spawn Probability=false, UseLoopCountLimit=false)
  ]
  particleSpawnScript [&InitializeParticle(
    Color=<rgba 0.0832 0.8022 2 1>
    Lifetime=0.3
    Sprite Size Mode=::NewEnumerator3
    Uniform Sprite Size=32
    UsePositionOffset=false
    Write Material Random=true
   )]
  particleUpdateScript [
   &SkeletalMeshLocation(MaskSpawnGroup=false, Mesh Sampling Type=::NewEnumerator5, Skeletal Mesh=$System.Skeletal Mesh, Socket Sampling Mode=::NewEnumerator5)
   &ParticleState()
   &ScaleSpriteSize(Uniform Curve Sprite Scale=DataInterface() {propertyValues {Curve {keys [{interpMode=RCIM_Cubic, leaveTangent=3.1532, tangentMode=RCTM_Break, time=0, value=0}, {tangentMode=RCTM_User, time=0.7, value=1}]}}})
   &ScaleColor(
    Linear Color Curve=DataInterface() {propertyValues {
      AlphaCurve {keys [<at 0.0539 1>,<at 0.6865 0.2>,<at 1 0>]}
      BlueCurve {keys [<at 0 1>,<at 1 1>]}
      GreenCurve {keys [<at 0 1>,<at 1 1>]}
      RedCurve {keys [<at 0 1>,<at 1 1>]}
     }}
    Scale Mode=::NewEnumerator2
   )
  ]
  rendererClasses [&NiagaraSpriteRendererProperties]
  renderers [{rendererClass=&NiagaraSpriteRendererProperties}]
 }
}
summary {
 systemName=NS_Backpack_AstroAnode_Burst
 emitters [
  {emitterName=Burst, rendererClasses [&NiagaraMeshRendererProperties, &NiagaraSpriteRendererProperties]}
  {emitterName=Minimal, rendererClasses [&NiagaraSpriteRendererProperties]}
 ]
 userVariables [
  User.bIsFrontEnd {type=NiagaraBool}
  User.bIsFrontEndPreview {type=NiagaraBool}
  User.bIsSwinging {type=NiagaraBool}
 ]
}
