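# Idle Niagara effect for the AstroAnode backpack, emitting blue glowing cubes with mesh and sprite renderers around its skeletal mesh.
# GPU CubesHighEpic re-samples the mesh each update, kills particles by a UV-sampled mask texture, then spins and lifts them along normals late in life.
# CubesMedLow is a sparser variant, spawning in a box with easing drift.
# LandingCubes uses a timeline-driven rate curve for a short burst that falls under gravity and drag.
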NiagaraSystem /BRCosmetics/Accessories/FORT_Backpacks/Backpack_AstroAnode/FX/NS_Backpack_AstroAnode_Idle
dependencies {
 usedDataInterfaces [
  &NiagaraDataInterfaceCurve
  &NiagaraDataInterfaceColorCurve
  /Script/Niagara.NiagaraDataInterfaceTexture
  &NiagaraDataInterfaceSkeletalMesh
 ]
 usedDynamicInputs []
 usedModules [
  &self:SystemSpawnScript.SystemScriptSource.SystemScriptGraph.NiagaraNodeAssignment_0.SetVariables_#1
  &SystemState
  &Timeline
 ]
}
emitters {
 CubesHighEpic {
  simTarget=GPUComputeSim
  emitterUpdateScript [
   &EmitterState()
   &SpawnRate(SpawnRate=18, Use Spawn Probability=false)
  ]
  particleSpawnScript [&InitializeParticle(
    Color=<rgba 0.046 0.8162 1.5 2>
    Lifetime Mode=::NewEnumerator1
    Mesh Scale Mode=::NewEnumerator1
    Mesh Uniform Scale Max=1.2
    Sprite Size Mode=::NewEnumerator3
    UsePositionOffset=false
    Write Material Random=true
   )]
  particleUpdateScript [
   &ParticleState()
   &SkeletalMeshLocation(
    MaskSpawnGroup=false
    Mesh Sampling Type=::NewEnumerator7
    OffsetAlongNormal=false
    Orientation Sampling=::NewEnumerator1
    OverrideRandomSeed=false
    Skeletal Mesh=$System.Skeletal Mesh
   ) {module=SkeletalMeshLocation002}
   &ScaleMeshSize(Scale Factor=&VectorFromFloat(Value=&FloatFromCurve(FloatCurve=DataInterface() {propertyValues {Curve {keys [
          {interpMode=RCIM_Cubic, leaveTangent=0.5222, tangentMode=RCTM_Break, time=0, value=0}
          {arriveTangent=4.1279, interpMode=RCIM_Cubic, tangentMode=RCTM_Break, time=0.4, value=1}
          {arriveTangent=-0.111, interpMode=RCIM_Cubic, leaveTangent=-3.88, tangentMode=RCTM_Break, time=0.8, value=1}
          <at 1 0.4 RCIM_Cubic>
         ]}}})))
   &ScaleColor(
    Scale Alpha=&FloatFromCurve(FloatCurve=DataInterface() {propertyValues {Curve {keys [{interpMode=RCIM_Cubic, leaveTangent=-7.9721, tangentMode=RCTM_Break, time=0.7, value=1}, <at 1 0>]}}})
    Scale RGB=<0 0 0>
    ScaleA=true
    ScaleRGB=true
   )
   /SetVariables(Particles.GlowColor=&ColorFromCurve(ColorCurve=DataInterface() {propertyValues {
       AlphaCurve {keys [
         <at 0.0011 0>
         {arriveTangent=0.5972, interpMode=RCIM_Cubic, leaveTangent=0.5972, time=0.4528, value=0.5}
         <at 0.8382 0.5>
         <at 1 0 RCIM_Cubic>
        ]}
       BlueCurve {keys [<at 0 2 RCIM_Cubic>,<at 1 0.4167 RCIM_Cubic>]}
       GreenCurve {keys [<at 0 0.8564 RCIM_Cubic>,<at 1 0.0401 RCIM_Cubic>]}
       RedCurve {keys [<at 0 0 RCIM_Cubic>,<at 1 0.0561 RCIM_Cubic>]}
      }})) {isSetParameters=true, module=SetVariables_#2}
   &UpdateMeshOrientation(
    Rotation Rate=&FloatFromCurve(
     FloatCurve=DataInterface() {propertyValues {Curve {keys [{interpMode=RCIM_Cubic, tangentMode=RCTM_Break, time=0.5, value=0}, {arriveTangent=2.1205, tangentMode=RCTM_User, time=1, value=1}]}}}
     Scale Curve=1.2
    )
    Rotation Vector=&RandomRangeVector(Minimum=<-1 -1 -1>, Maximum=<1 1 1>)
   )
   /Niagara/Modules/Update/Textures/SampleTexture(
    Texture=DataInterface() {propertyValues {
      Texture=&/T_Backpack_AstroAnode_FXMask
      TextureUserParameter {parameter {name=None, typeDefHandle {registeredTypeIndex=2}, varData []}}
     }}
    UV:Vector2f=$Particles.SkeletalMeshLocation002.SampledUV
   )
   &KillParticles(Kill Particles=&SetBoolByFloatComparison(A=$Particles.SampleTexture.SampledR, B=0.01, Comparison Type=::NewEnumerator6))
   /SetVariables(Particles.PrevPos:NiagaraPosition=$Particles.Position) {isSetParameters=true}
   /SetVariables(Particles.PosOffsetted:NiagaraPosition=&AddVectorToPosition(
     Position=$Particles.PrevPos
     Vector=&Multiply_VectorByFloat(Vector=$Particles.SkeletalMeshLocation002.SampledNormal, Float=15)
    )) {isSetParameters=true}
   /SetVariables(Particles.Position=&Lerp_Vector(
     A=Linked() {linkedVariable {name=Particles.PrevPos, type=NiagaraPosition}}
     B=Linked() {linkedVariable {name=Particles.PosOffsetted, type=NiagaraPosition}}
     Alpha=&FloatFromCurve(FloatCurve=DataInterface() {propertyValues {Curve {keys [{interpMode=RCIM_Cubic, tangentMode=RCTM_Break, time=0.6, value=0}, {arriveTangent=3.542, tangentMode=RCTM_User, time=1, value=1}]}}})
    )) {isSetParameters=true}
  ]
  rendererClasses [&NiagaraMeshRendererProperties, &NiagaraSpriteRendererProperties]
  renderers [
   {rendererClass=&NiagaraMeshRendererProperties, rendererIndex=0}
   {rendererClass=&NiagaraSpriteRendererProperties, rendererIndex=1}
  ]
 }
 CubesMedLow {
  emitterUpdateScript [
   &EmitterState()
   &SpawnRate(SpawnRate=5, Use Spawn Probability=false)
  ]
  particleSpawnScript [
   &InitializeParticle(
    Color=<rgba 0.046 0.8162 1.5 2>
    Lifetime Mode=::NewEnumerator1
    Mesh Scale Mode=::NewEnumerator1
    Mesh Uniform Scale Max=1.1
    Mesh Uniform Scale Min=0.8
    Sprite Size Mode=::NewEnumerator3
    UsePositionOffset=false
    Write Material Random=true
   )
   &SkeletalMeshLocation(Bone Sampling Mode=::NewEnumerator4, MaskSpawnGroup=false, Skeletal Mesh=$System.Skeletal Mesh) {module=SkeletalMeshLocation002}
   &ShapeLocation(Box Size=<30 10 30>, MaskSpawnGroup=false, Override Seed=false, Shape Primitive=::NewEnumerator2, Surface Only=false, UseOwnerScale=true)
   &AddVelocity(Velocity=<0 -25 10>)
  ]
  particleUpdateScript [
   /SetVariables(Particles.Velocity=&Lerp_Vector(
     A=<0 0 0>
     B=$Particles.Initial.Velocity
     Alpha=&FloatFromCurve(FloatCurve=DataInterface() {propertyValues {Curve {keys [{interpMode=RCIM_Cubic, leaveTangent=1.0829, tangentMode=RCTM_Break, time=0.4, value=0.1}, <at 0.6 0.4>]}}})
    )) {isSetParameters=true}
   &SolveForcesAndVelocity(Clamp Velocity=false, Limit Acceleration=false)
   &ParticleState()
   &ScaleMeshSize(Scale Factor=&VectorFromFloat(Value=&FloatFromCurve(FloatCurve=DataInterface() {propertyValues {Curve {keys [
          {interpMode=RCIM_Cubic, leaveTangent=0.5222, tangentMode=RCTM_Break, time=0, value=0}
          {arriveTangent=4.1279, interpMode=RCIM_Cubic, tangentMode=RCTM_Break, time=0.4, value=1}
          {arriveTangent=-0.111, interpMode=RCIM_Cubic, leaveTangent=-3.88, tangentMode=RCTM_Break, time=0.6, value=1}
          <at 1 0.4 RCIM_Cubic>
         ]}}})))
   &ScaleColor(
    Scale Alpha=&FloatFromCurve(FloatCurve=DataInterface() {propertyValues {Curve {keys [{interpMode=RCIM_Cubic, leaveTangent=-1.882, tangentMode=RCTM_Break, time=0.3, value=1}, <at 1 0>]}}})
    Scale RGB=<0 0 0>
    ScaleA=true
    ScaleRGB=true
   )
   /SetVariables(Particles.GlowColor=&ColorFromCurve(ColorCurve=DataInterface() {propertyValues {
       AlphaCurve {keys [<at 0.0011 0>,<at 0.4528 0.5 RCIM_Cubic>,<at 1 0 RCIM_Cubic>]}
       BlueCurve {keys [<at 0 2 RCIM_Cubic>,<at 1 0.4167 RCIM_Cubic>]}
       GreenCurve {keys [<at 0 0.8564 RCIM_Cubic>,<at 1 0.0401 RCIM_Cubic>]}
       RedCurve {keys [<at 0 0 RCIM_Cubic>,<at 1 0.0561 RCIM_Cubic>]}
      }})) {isSetParameters=true, module=SetVariables_#2}
  ]
  rendererClasses [&NiagaraMeshRendererProperties, &NiagaraSpriteRendererProperties]
  renderers [
   {rendererClass=&NiagaraMeshRendererProperties, rendererIndex=0}
   {rendererClass=&NiagaraSpriteRendererProperties, rendererIndex=1}
  ]
 }
 LandingCubes {
  emitterUpdateScript [
   &EmitterState()
   &SpawnRate(
    SpawnRate=&FloatFromCurve(
     FloatCurve=DataInterface() {propertyValues {Curve {keys [
         {interpMode=RCIM_Cubic, leaveTangent=48.3547, tangentMode=RCTM_Break, time=0, value=0}
         {arriveTangent=7.876, interpMode=RCIM_Cubic, leaveTangent=-21.7956, tangentMode=RCTM_Break, time=0.03, value=1}
         <at 0.06 0>
        ]}}}
     CurveIndex=$System.Timeline.Time
     Scale Curve=150
    )
    Use Spawn Probability=false
   )
  ]
  particleSpawnScript [
   &InitializeParticle(
    Color=<rgba 0.046 0.8162 1.5 2>
    Lifetime Max=0.5
    Lifetime Min=0.35
    Lifetime Mode=::NewEnumerator1
    Mesh Scale Mode=::NewEnumerator1
    Mesh Uniform Scale Max=1.1
    Mesh Uniform Scale Min=0.8
    Sprite Size Mode=::NewEnumerator3
    UsePositionOffset=false
    Write Material Random=true
   )
   &SkeletalMeshLocation(Bone Sampling Mode=::NewEnumerator4, MaskSpawnGroup=false, Skeletal Mesh=$System.Skeletal Mesh) {module=SkeletalMeshLocation002}
   &ShapeLocation(
    Box Midpoint=<0.5 0.5 1>
    Box Size=<40 10 20>
    MaskSpawnGroup=false
    Override Seed=false
    Shape Origin=$Particles.SkeletalMeshLocation002.SampledPosition
    Shape Primitive=::NewEnumerator2
    Surface Only=false
    UseOwnerScale=true
   )
  ]
  particleUpdateScript [
   &Drag(Drag=3, Use Linear Drag=true, Use Rotational Drag=true)
   &GravityForce(Gravity=<0 0 -200>)
   &SolveForcesAndVelocity(Clamp Velocity=false, Limit Acceleration=false)
   &ParticleState()
   &UpdateMeshOrientation(Rotation Rate=0.6, Rotation Vector=&RandomRangeVector(Minimum=<0 0 0>, Maximum=<1 1 1>))
   &ScaleMeshSize(Scale Factor=&VectorFromFloat(Value=&FloatFromCurve(FloatCurve=DataInterface() {propertyValues {Curve {keys [
          {interpMode=RCIM_Cubic, leaveTangent=0.5222, tangentMode=RCTM_Break, time=0, value=0.4}
          {arriveTangent=-0.245, interpMode=RCIM_Cubic, tangentMode=RCTM_Break, time=0.3, value=1}
          {arriveTangent=-0.111, interpMode=RCIM_Cubic, leaveTangent=-3.88, tangentMode=RCTM_Break, time=0.6, value=1}
          <at 1 0.4 RCIM_Cubic>
         ]}}})))
   &ScaleColor(
    Scale Alpha=&FloatFromCurve(FloatCurve=DataInterface() {propertyValues {Curve {keys [{interpMode=RCIM_Cubic, leaveTangent=-1.882, tangentMode=RCTM_Break, time=0.3, value=1}, <at 1 0>]}}})
    Scale RGB=<0 0 0>
    ScaleA=true
    ScaleRGB=true
   )
   /SetVariables(Particles.GlowColor=&ColorFromCurve(ColorCurve=DataInterface() {propertyValues {
       AlphaCurve {keys [
         <at 0.0011 0>
         <at 0.0955 0.4>
         {arriveTangent=-0.4422, interpMode=RCIM_Cubic, leaveTangent=-0.4422, time=0.4528, value=0.5}
         <at 1 0 RCIM_Cubic>
        ]}
       BlueCurve {keys [<at 0 2 RCIM_Cubic>,<at 1 0.4167 RCIM_Cubic>]}
       GreenCurve {keys [<at 0 0.8564 RCIM_Cubic>,<at 1 0.0401 RCIM_Cubic>]}
       RedCurve {keys [<at 0 0 RCIM_Cubic>,<at 1 0.0561 RCIM_Cubic>]}
      }})) {isSetParameters=true, module=SetVariables_#2}
  ]
  rendererClasses [&NiagaraMeshRendererProperties, &NiagaraSpriteRendererProperties]
  renderers [
   {rendererClass=&NiagaraMeshRendererProperties, rendererIndex=0}
   {rendererClass=&NiagaraSpriteRendererProperties, rendererIndex=1}
  ]
 }
}
summary {
 systemName=NS_Backpack_AstroAnode_Idle
 emitters [
  {
   emitterName=CubesHighEpic
   simTarget=GPUComputeSim
   rendererClasses [&NiagaraMeshRendererProperties, &NiagaraSpriteRendererProperties]
  }
  {emitterName=CubesMedLow, rendererClasses [&NiagaraMeshRendererProperties, &NiagaraSpriteRendererProperties]}
  {emitterName=LandingCubes, rendererClasses [&NiagaraMeshRendererProperties, &NiagaraSpriteRendererProperties]}
 ]
 userVariables [
  User.bIsLanded {type=NiagaraBool}
  User.bIsSwinging {type=NiagaraBool}
 ]
}
