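# Character material for the AstroAnode back bling, adding an emissive grid effect and a single-digit kill counter display.
# It layers effects between the shared CharacterParent_2023 PreFX and PostFX functions, flattening base color, metallic, specular and roughness in masked areas.
# Grid tiles pulse on offset sine waves.
# Kill_Number picks a digit from a flipbook atlas, combined with scrolling static and sine-driven glitch effects, placed by a counter mask.
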Material /BRCosmetics/Accessories/FORT_Backpacks/Backpack_AstroAnode/Materials/M_Backpack_AstroAnode
graph {
 properties {
  bUseMaterialAttributes=true
  bUsedWithClothing=true
  bUsedWithMeshDeformer=true
  bUsedWithMorphTargets=true
  bUsedWithNiagaraMeshParticles=true
  bUsedWithSkeletalMesh=true
  pixelDepthOffsetMode=PDOM_Legacy
  previewMesh=/BRCosmetics/Characters/Player/Female/Medium/Bodies/F_MED_Astronaut_01/Meshes/Parts/F_Med_Astronaut_01_Backpack
  shadingModel=MSM_FromMaterialExpression
  usedShadingModels=Default Lit | Subsurface | Clear Coat | Subsurface Profile | Cloth
 }
 nodes [
  n1 = MaterialFunctionCall() {materialFunction=&MF_CharacterParent_2023_PreFX}
  n2 = GetMaterialAttributes(None=@n1.Result) {attributeGetTypes [
    #1
    #2
    #3
    #4
    #5
   ]}
  n3 = Add(
   A=Multiply(A=Time() {period=1}, B=ScalarParameter() {defaultValue=0.5, group=GridEffect, parameterName=Tile_Speed})
   B=Multiply(A=NamedRerouteUsage(), B=ScalarParameter() {defaultValue=1, group=GridEffect, parameterName=Tile_Rand_Offset})
  ) {constA=1}
  n4 = LinearInterpolate(
   A=ScalarParameter() {defaultValue=0.1, group=GridEffect, parameterName=Tile_Intensity_Min}
   B=ScalarParameter() {defaultValue=1, group=GridEffect, parameterName=Tile_Intensity_Max}
   Alpha=ConstantBiasScale(None=Sine(None=@n3))
  )
  n5 = Step(X=Subtract(A=NamedRerouteUsage(), B=NamedRerouteUsage())) {constY=0.05}
  n6 = LinearInterpolate(
   A=Multiply(A=VectorParameter() {group=GridEffect, parameterName=Tile_Color, defaultValue=<rgba 0 0.7729 1 1>}.RGB, B=Multiply(A=@n4, B=@n5))
   B=Multiply(
    A=VectorParameter() {group=GridEffect, parameterName=Grid_Color, defaultValue=<rgba 0 1 0.8801 1>}.RGB
    B=ScalarParameter() {defaultValue=1, group=GridEffect, parameterName=Grid_Intensity}
   )
   Alpha=NamedRerouteUsage()
  )
  n7 = NamedRerouteUsage()
  n8 = TextureSampleParameter2D() {
   group=KillCounter
   parameterName=CounterMask
   samplerType=SAMPLERTYPE_Masks
   texture=/BRCosmetics/Accessories/FORT_Backpacks/Backpack_AstroAnode/Textures/T_AstroAnode_Backpack_Counter_Mask
  }
  n9 = OneMinus(None=Step(Y=NamedRerouteUsage()) {constX=0})
  n10 = SetMaterialAttributes!(
   MaterialAttributes=@n2.MaterialAttributes
   Emissive Color=LinearInterpolate(A=LinearInterpolate(B=@n6, Alpha=@n7), B=NamedRerouteUsage(), Alpha=@n8.R)
   Metallic=LinearInterpolate(A=@n2.Metallic, Alpha=@n9) {constB=0}
   Specular=LinearInterpolate(A=@n2.Specular, Alpha=@n9) {constB=0}
   Roughness=LinearInterpolate(A=@n2.Roughness, Alpha=@n9)
   Base Color=LinearInterpolate(A=@n2.Base Color, Alpha=@n9) {constB=0}
  ) {attributeSetTypes [
    #1
    #2
    #3
    #4
    #5
   ]}
  MaterialFunctionCall(Material=@n10, M=@n1.M, Pre Skinned Local Position=@n1.M, Distance Blend=@n1.M, Fresnel=@n1.M) {materialFunction=&MF_CharacterParent_2023_PostFX}
  n11 = Panner(
   Coordinate=Multiply(A=TextureCoordinate(), B=ScalarParameter() {defaultValue=1, group=KillCounter, parameterName=Counter_Static_Tiling})
   Speed=AppendVector(A=Constant(), B=ScalarParameter() {defaultValue=0.1, group=KillCounter, parameterName=Counter_Static_Speed})
  ) {speedY=0.5}
  NamedRerouteDeclaration(None=Multiply(
    A=TextureSample(UVs=@n11) {texture=/BRCosmetics/Accessories/FORT_Backpacks/Backpack_AstroAnode/Textures/T_AstroAnode_Static}.R
    B=VectorParameter() {group=KillCounter, parameterName=Counter_Static_Color, defaultValue=<rgba 0 0.7246 1 1>}.RGB
   )) {name=Static_Effect, nodeColor=<rgba 0.5961 0.7721 1 1>}
  n12 = ScalarParameter() {group=KillCounter, parameterName=Counter_Glitch}
  n13 = Multiply(A=Time() {period=1})
  n14 = TextureSample(UVs=Multiply(A=TextureCoordinate() {uTiling=0.5}, B=ScalarParameter() {defaultValue=1, group=KillCounter, parameterName=Glitch_Tiling})) {texture=/BRCosmetics/Accessories/FORT_Backpacks/Backpack_AstroAnode/Textures/T_AstroAnode_Glitch_Effect}
  n15 = ConstantBiasScale(None=Sine(None=Add(A=@n13, B=@n14.B) {constA=1}))
  NamedRerouteDeclaration(None=Multiply(A=Multiply(A=@n12, B=@n15) {constA=0.4}, B=Constant3Vector() {constant=<rgba 0 0.4141 1 1>})) {name=Glitch_Effect, nodeColor=<rgba 0 1 0.4141 1>}
  n16 = NamedRerouteUsage()
  n17 = Clamp(None=ScalarParameter() {group=KillCounter, parameterName=Kill_Number}) {maxDefault=9}
  n18 = Constant() {r=10}
  n19 = Constant() {r=16}
  n20 = Divide(A=Floor(None=Fmod(A=@n17, B=@n18)), B=@n19)
  n21 = Divide(A=Constant() {r=9}, B=@n19)
  n22 = Multiply(A=TextureCoordinate(), B=ScalarParameter() {defaultValue=1, group=KillCounter, parameterName=Counter_Tiling, sliderMin=1})
  n23 = AppendVector(
   A=ScalarParameter() {group=KillCounter, parameterName=Counter_U_Offset}
   B=ScalarParameter() {group=KillCounter, parameterName=Counter_V_Offset}
  )
  n24 = TextureObjectParameter() {
   group=KillCounter
   parameterName=NumberTex
   samplerType=SAMPLERTYPE_LinearColor
   texture=/BRCosmetics/Accessories/FORT_Backpacks/Backpack_AstroAnode/Textures/T_AstroAnode_Number_Atlas
  }
  n25 = TextureSample(
   UVs=MaterialFunctionCall(
    "Animation  Phase (0-1)"=Clamp(None=@n20, Max=@n21)
    Number of Rows=Constant() {r=4}
    Number of Columns=Constant() {r=4}
    UVs=Add(A=@n22, B=@n23)
   ) {materialFunction=&FlipBook}.UVs
   Tex=@n24
  ) {
   samplerSource=SSM_Wrap_WorldGroupSettings
   samplerType=SAMPLERTYPE_Masks
   texture=/BRCosmetics/Accessories/FORT_Backpacks/Backpack_M_MED_WarmShadeWeasel/Textures/T_CountBoardNumbers
  }
  n26 = Add(A=@n25.A, B=NamedRerouteUsage())
  n27 = Multiply(
   A=Multiply(
    A=VectorParameter() {group=KillCounter, parameterName=Counter_Color, defaultValue=<rgba 0 0.5423 1 1>}.RGB
    B=ScalarParameter() {defaultValue=1, group=KillCounter, parameterName=Counter_Intensity}
   )
   B=@n26
  )
  NamedRerouteDeclaration(None=Subtract(A=LinearInterpolate(A=@n16, B=@n27, Alpha=@n26), B=NamedRerouteUsage())) {name=Kill_Counter, nodeColor=<rgba 1 0.0857 0.8659 1>}
  NamedRerouteDeclaration(None=@n2.Emissive Color) {name=Emissive Color, nodeColor=<rgba 1 0.9232 0 1>}
  n28 = ComponentMask(None=@n1.Tech Art Mask) {g=true}
  n29 = ComponentMask(None=@n1.Tech Art Mask) {b=true}
  NamedRerouteDeclaration(None=Add(
    A=NamedRerouteDeclaration(None=@n28) {name=TA_Mask_G, nodeColor=<rgba 0.0347 1 0 1>}
    B=NamedRerouteDeclaration(None=@n29) {name=TA_Mask_B, nodeColor=<rgba 0 0.0319 1 1>}
   )) {name=TA_Mask_GB, nodeColor=<rgba 0 1 1 1>}
 ]
}
